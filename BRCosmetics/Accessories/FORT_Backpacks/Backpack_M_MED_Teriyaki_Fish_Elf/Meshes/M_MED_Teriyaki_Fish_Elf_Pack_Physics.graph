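# Physics asset for the Teriyaki Fish Elf backpack cosmetic, defining jiggle bodies and joint limits for its small character figure.
# Capsule bodies cover the pack root, head, mouth, eyeballs, head and neck tassels, and two-segment arms and legs.
# Limbs chain from the pack root, while mouth, eyeballs and head tassels hang off the head.
# Both segments of each neck tassel constrain directly to the root; mouth and head tassels swing on one axis.
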PhysicsAsset /BRCosmetics/Accessories/FORT_Backpacks/Backpack_M_MED_Teriyaki_Fish_Elf/Meshes/M_MED_Teriyaki_Fish_Elf_Pack_Physics
bodies {
 dyn_L_arm_jnt_1 [{shapeType=Capsule, length=1.5, radius=2, center=<1.7381 0 0.0511>, rotation=<rot -88.3173 0 0>}]
 dyn_L_arm_jnt_2 [{shapeType=Capsule, length=1.5, radius=2, center=<2.7837 0.0001 0.7665>, rotation=<rot -74.4572 0 0>}]
 dyn_L_eyeball_jnt [{shapeType=Capsule, length=1, radius=1}]
 dyn_L_head_tassle_jnt [{shapeType=Capsule, length=12, radius=1, center=<3.2772 -0.2149 0.272>, rotation=<rot 0 88.4071 -83.9734>}]
 dyn_L_leg_jnt_1 [{shapeType=Capsule, length=2, radius=1.5, center=<2.5552 0 0>, rotation=<rot 90 0 0>}]
 dyn_L_leg_jnt_2 [{shapeType=Capsule, length=1.5, radius=1.5, center=<2.3026 0 0>, rotation=<rot 90 0 0>}]
 dyn_L_neck_tassle_jnt_1 [{shapeType=Capsule, length=2.5, radius=1.7, center=<0.6643 -0.0002 0>, rotation=<rot 90 0 0>}]
 dyn_L_neck_tassle_jnt_2 [{shapeType=Capsule, length=2.5, radius=1.7, center=<2.8215 -0.0001 0.6761>, rotation=<rot -76 0 0>}]
 dyn_R_arm_jnt_1 [{shapeType=Capsule, length=1.5, radius=2, center=<1.7381 0 0.0511>, rotation=<rot -88.3173 0 0>}]
 dyn_R_arm_jnt_2 [{shapeType=Capsule, length=1.5, radius=2, center=<2.7837 0.0001 0.7665>, rotation=<rot -74.4572 0 0>}]
 dyn_R_eyeball_jnt [{shapeType=Capsule, length=1, radius=1}]
 dyn_R_head_tassle_jnt [{shapeType=Capsule, length=12, radius=1, center=<3.2772 0.215 0.272>, rotation=<rot 0 91.5929 -83.9734>}]
 dyn_R_leg_jnt_1 [{shapeType=Capsule, length=2, radius=1.5, center=<2.5552 0 0>, rotation=<rot 90 0 0>}]
 dyn_R_leg_jnt_2 [{shapeType=Capsule, length=1.5, radius=1.5, center=<2.3026 0 0>, rotation=<rot 90 0 0>}]
 dyn_R_neck_tassle_jnt_1 [{shapeType=Capsule, length=2.5, radius=1.7, center=<0.6643 -0.0002 0>, rotation=<rot 90 0 0>}]
 dyn_R_neck_tassle_jnt_2 [{shapeType=Capsule, length=2.5, radius=1.7, center=<2.8215 -0.0001 0.6761>, rotation=<rot -76 0 0>}]
 dyn_head_jnt [{shapeType=Capsule, length=4, radius=7, center=<-6.9106 0 0>}]
 dyn_mouth_jnt [{shapeType=Capsule, length=2, radius=2}]
 dyn_pack_master_jnt [{shapeType=Capsule, length=2, radius=2}]
}
constraints [
 {bone1Name=dyn_R_arm_jnt_1, bone2Name=dyn_pack_master_jnt, swing1LimitDegrees=30, swing2LimitDegrees=35, twistLimitDegrees=3}
 {bone1Name=dyn_R_arm_jnt_2, bone2Name=dyn_R_arm_jnt_1, swing1LimitDegrees=25, swing2LimitDegrees=20, twistLimitDegrees=3}
 {bone1Name=dyn_L_arm_jnt_1, bone2Name=dyn_pack_master_jnt, swing1LimitDegrees=30, swing2LimitDegrees=35, twistLimitDegrees=3}
 {bone1Name=dyn_L_arm_jnt_2, bone2Name=dyn_L_arm_jnt_1, swing1LimitDegrees=25, swing2LimitDegrees=20, twistLimitDegrees=3}
 {bone1Name=dyn_head_jnt, bone2Name=dyn_pack_master_jnt, swing1LimitDegrees=6, swing2LimitDegrees=3.5, twistLimitDegrees=3, twistMotion=Locked}
 {
  bone1Name=dyn_mouth_jnt
  bone2Name=dyn_head_jnt
  swing1LimitDegrees=13
  swing2LimitDegrees=45
  swing2Motion=Locked
  twistLimitDegrees=45
  twistMotion=Locked
 }
 {bone1Name=dyn_L_eyeball_jnt, bone2Name=dyn_head_jnt, swing1LimitDegrees=10, swing2LimitDegrees=10, twistLimitDegrees=10}
 {bone1Name=dyn_R_eyeball_jnt, bone2Name=dyn_head_jnt, swing1LimitDegrees=10, swing2LimitDegrees=10, twistLimitDegrees=10}
 {bone1Name=dyn_L_leg_jnt_1, bone2Name=dyn_pack_master_jnt, swing1LimitDegrees=20, swing2LimitDegrees=20, twistLimitDegrees=10}
 {bone1Name=dyn_L_leg_jnt_2, bone2Name=dyn_L_leg_jnt_1, swing1LimitDegrees=20, swing2LimitDegrees=20, twistLimitDegrees=10}
 {bone1Name=dyn_R_leg_jnt_1, bone2Name=dyn_pack_master_jnt, swing1LimitDegrees=20, swing2LimitDegrees=20, twistLimitDegrees=10}
 {bone1Name=dyn_R_leg_jnt_2, bone2Name=dyn_R_leg_jnt_1, swing1LimitDegrees=20, swing2LimitDegrees=20, twistLimitDegrees=10}
 {
  bone1Name=dyn_L_neck_tassle_jnt_1
  bone2Name=dyn_pack_master_jnt
  swing1LimitDegrees=8
  swing2LimitDegrees=15
  twistLimitDegrees=45
  twistMotion=Locked
 }
 {
  bone1Name=dyn_L_neck_tassle_jnt_2
  bone2Name=dyn_pack_master_jnt
  swing1LimitDegrees=8
  swing2LimitDegrees=15
  twistLimitDegrees=45
  twistMotion=Locked
 }
 {
  bone1Name=dyn_R_neck_tassle_jnt_1
  bone2Name=dyn_pack_master_jnt
  swing1LimitDegrees=8
  swing2LimitDegrees=15
  twistLimitDegrees=45
  twistMotion=Locked
 }
 {
  bone1Name=dyn_R_neck_tassle_jnt_2
  bone2Name=dyn_pack_master_jnt
  swing1LimitDegrees=8
  swing2LimitDegrees=15
  twistLimitDegrees=45
  twistMotion=Locked
 }
 {
  bone1Name=dyn_L_head_tassle_jnt
  bone2Name=dyn_head_jnt
  swing1LimitDegrees=45
  swing1Motion=Locked
  swing2LimitDegrees=30
  twistLimitDegrees=45
  twistMotion=Locked
 }
 {
  bone1Name=dyn_R_head_tassle_jnt
  bone2Name=dyn_head_jnt
  swing1LimitDegrees=45
  swing1Motion=Locked
  swing2LimitDegrees=30
  twistLimitDegrees=45
  twistMotion=Locked
 }
]
